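# Idle cosmetic effect for the FunkyFeathers Coin character skin, flashing short-lived DiamondBlings sparkle sprites on the body.
# Six probabilistic spawn groups map to ten socket indices, each placing particles at random hand-authored offsets that follow the socket.
# Sprites pop in via size and brightness curves, with alpha dimmed by time of day.
# Particles in hand, foot, head/hat and necklace groups are killed when those parts are hidden.
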NiagaraSystem /BRCosmetics/Characters/Player/Male/Medium/Bodies/M_MED_FunkyFeather_Stripes/Skins/Coin/FX/NS_Character_FunkyFeathers_Coin_Idle
dependencies {
 usedDataInterfaces [&NiagaraDataInterfaceArrayFloat3, &NiagaraDataInterfaceCurve, &NiagaraDataInterfaceSocketReader]
 usedDynamicInputs [&RandomRangeLinearColor, &MakeIntFromFloat]
 usedModules [
  &self:SystemSpawnScript.SystemScriptSource.SystemScriptGraph.NiagaraNodeAssignment_3.SetVariables_#1
  &SystemState
  &self:SystemSpawnScript.SystemScriptSource.SystemScriptGraph.NiagaraNodeAssignment_8.SetVariables_#2
 ]
}
emitters {DiamondBlings {
  emitterUpdateScript [
   &EmitterState(Inactive Response=::NewEnumerator1, Life Cycle Mode=::NewEnumerator1, Loop Duration=1, UseLoopDelay=false)
   &SpawnRate(Spawn Probability=0.6, Use Spawn Probability=true)
   &SpawnRate(Spawn Probability=0.7, SpawnGroup=1, Use Spawn Probability=true) {module=SpawnRate005}
   &SpawnRate(Spawn Probability=0.6, SpawnGroup=2, SpawnRate=2, Use Spawn Probability=true) {module=SpawnRate001}
   &SpawnRate(Spawn Probability=0.45, SpawnGroup=3, Use Spawn Probability=true) {module=SpawnRate004}
   &SpawnRate(Spawn Probability=0.75, SpawnGroup=4, Use Spawn Probability=true) {module=SpawnRate002}
   &SpawnRate(Spawn Probability=0.6, SpawnGroup=5, Use Spawn Probability=true) {module=SpawnRate003}
  ]
  particleSpawnScript [
   /SetVariables(Particles.PartGroup:NiagaraInt32=$Emitter.SpawnGroup) {isSetParameters=true}
   /SetVariables(Particles.SpawnGroup=&MergeHelper_IntSwitch(
     0=0
     1=&RandomRangeInt(Minimum=1, Maximum=2)
     2=&RandomRangeInt(Minimum=3, Maximum=5)
     3=6
     4=&RandomRangeInt(Minimum=7, Maximum=8)
     5=9
     Index=$Emitter.SpawnGroup
     Num=::NewEnumerator4
    )) {isSetParameters=true}
   &InitializeParticle(
    Lifetime=&RandomRangeFloat(Minimum=0.18, Maximum=0.225)
    Sprite Rotation Angle=&RandomRangeFloat(Minimum=-0.055, Maximum=0.055)
    Sprite Rotation Mode=::NewEnumerator1
    Sprite Size Mode=::NewEnumerator3
    Sprite UV Mode=::NewEnumerator3
    Uniform Sprite Size=&RandomRangeFloat(Minimum=11, Maximum=20)
    UsePositionOffset=false
    Write Material Random=true
   )
   /SetVariables(Particles.RotationDirection:NiagaraFloat=&UniformAOrBFloat(A=-1, Evaluation Type=::NewEnumerator0, Recalculate Random Each Loop=true, Override Seed=false)) {isSetParameters=true}
   &CameraOffset(Camera Offset Amount=4)
  ]
  particleUpdateScript [
   &ParticleState()
   &SocketLocation(
    Interpolate Socket Sampling=true
    SamplePosOffset=true
    Sampled Position Offset=&MergeHelper_VectorSwitch(
     0=&SelectVectorFromArray(
      Array Sampling Mode=::NewEnumerator0
      Direct Array Index=?
      Vector Selection Array=DataInterface() {propertyValues {FloatData [
         <-19.3303 1.8771 -3.7103>
         <-20.4456 2.2585 -3.4315>
         <-21.788 -1.4087 -1.9452>
         <-20.6272 -2.5146 -1.8955>
         <-20.8043 -2.6189 -2.3668>
         <-18.5367 -2.3319 -2.4226>
         <-17.7626 -0.2698 -3.2636>
         <-19.8456 2.9595 0.3495>
         <-18.8356 -1.9018 1.8226>
         <-20.3297 -2.7945 1.4608>
         <-20.3993 -3.1524 -1.4093>
        ]}}
      Random Seed=0
     )
     1=&SelectVectorFromArray(
      Array Sampling Mode=::NewEnumerator0
      Direct Array Index=?
      Vector Selection Array=DataInterface() {propertyValues {FloatData [
         <3.1275 9.0822 7.0127>
         <2.2286 9.0573 6.5638>
         <-0.3714 11.025 5.2964>
         <-6.3611 12.7862 0.8352>
         <-4.6451 11.6674 -3.7673>
         <-1.2141 11.5759 -3.9304>
         <2.8066 9.9152 -5.5246>
         <5.859 7.8994 -6.833>
         <5.6956 7.906 7.295>
         <-7.0279 13.1889 -0.0103>
        ]}}
      Random Seed=0
     )
     2=&SelectVectorFromArray(
      Array Sampling Mode=::NewEnumerator0
      Direct Array Index=?
      Vector Selection Array=DataInterface() {propertyValues {FloatData [
         <8.3921 -1.8385 -0.1172>
         <8.3587 0.8953 0.6299>
         <8.2512 -1.572 0.6723>
         <1.7136 0.385 0.4493>
         <5.0027 0.1442 0.6128>
         <7.2817 2.1907 0.5431>
        ]}}
      Random Seed=0
     )
     3=&SelectVectorFromArray(
      Array Sampling Mode=::NewEnumerator0
      Direct Array Index=?
      Vector Selection Array=DataInterface() {propertyValues {FloatData [
         <-17.2571 -5.8934 -0.0855>
         <-18.5367 -5.5847 5.0833>
         <-13.9084 -5.1273 5.4963>
         <-10.328 -3.2992 7.134>
         <-9.4175 -3.4454 7.0419>
         <-4.9011 -5.5604 4.0789>
         <-3.3641 3.4947 -4.4281>
         <-7.1356 5.429 -3.2938>
         <-15.5379 -0.8345 8.2668>
         <-14.6269 4.6187 7.7081>
         <-12.3813 3.6145 8.3205>
         <-10.0198 2.5981 8.6294>
        ]}}
      Random Seed=0
     )
     4=&SelectVectorFromArray(
      Array Sampling Mode=::NewEnumerator0
      Direct Array Index=?
      Vector Selection Array=DataInterface() {propertyValues {FloatData [
         <17.6111 5.614 -3.7609>
         <13.0633 3.5959 -7.2691>
         <5.0934 5.9947 -2.587>
         <18.4565 0.8936 5.0497>
         <15.6898 -4.0368 4.8654>
         <10.1898 -6.8621 -4.937>
         <9.9242 -7.4778 -3.362>
         <10.7442 -6.6207 -5.3754>
         <17.3087 -7.0111 -4.5724>
        ]}}
      Random Seed=0
     )
     5=&SelectVectorFromArray(
      Array Sampling Mode=::NewEnumerator0
      Direct Array Index=?
      Vector Selection Array=DataInterface() {propertyValues {FloatData [
         <7.552 -10.8838 -12.8101>
         <3.2923 -9.4596 -14.4772>
         <-5.7604 -7.0712 -16.7792>
         <3.2212 5.9731 -14.3339>
         <9.4025 2.898 -16.3652>
         <-5.9719 7.6089 -12.8348>
         <-5.5591 6.2261 -14.3267>
         <-11.3907 -3.6572 -18.2149>
         <8.2135 -13.0746 -7.1272>
         <7.4737 -13.5614 -4.6703>
         <-0.6657 -14.0682 -2.8112>
         <9.9425 -13.1455 3.7392>
         <-14.2864 -2.0029 18.6151>
         <-15.9549 -14.4188 -7.1713>
         <-16.2523 -13.1363 -10.2622>
         <-12.1005 -13.4656 -9.1192>
         <-8.4373 -13.5112 3.178>
         <-12.0596 -14.4783 -3.049>
         <-15.2126 -14.5439 -5.7409>
         <-1.7071 0.0748 17.3786>
         <-7.4795 -0.071 17.8044>
         <-12.7166 -1.5556 18.4453>
         <-12.8312 3.5876 17.4287>
         <-1.8236 11.0742 7.6874>
         <-3.7523 11.8917 4.5165>
         <-10.1296 11.9132 -3.6359>
         <-16.7371 12.0067 -4.1733>
         <-9.6349 9.3287 10.7178>
         <-16.8732 10.126 9.4211>
         <11.596 11.4751 4.2823>
        ]}}
      Random Seed=0
     )
     6=&SelectVectorFromArray(
      Array Sampling Mode=::NewEnumerator0
      Direct Array Index=?
      Vector Selection Array=DataInterface() {propertyValues {FloatData [
         <15.6205 -4.3765 6.4501>
         <9.2678 -4.943 6.6358>
         <11.2245 -4.7492 7.1839>
         <11.3066 -3.9339 7.5097>
         <12.3055 -3.314 8.3697>
         <10.8398 -0.1452 9.074>
        ]}}
      Random Seed=0
     )
     7=&SelectVectorFromArray(
      Array Sampling Mode=::NewEnumerator0
      Direct Array Index=?
      Vector Selection Array=DataInterface() {propertyValues {FloatData [
         <4.0618 -21.3497 -0.1079>
         <3.7142 -17.7417 -3.2426>
         <3.1827 -17.0073 -2.6947>
         <3.3198 -16.1354 -3.234>
         <3.0142 -17.1625 2.3656>
         <3.2481 -15.3379 3.7147>
         <-3.1934 -5.5279 1.0596>
         <-2.4167 -5.9927 -0.236>
        ]}}
      Random Seed=0
     )
     8=&SelectVectorFromArray(
      Array Sampling Mode=::NewEnumerator0
      Direct Array Index=?
      Vector Selection Array=DataInterface() {propertyValues {FloatData [
         <3.2341 5.4759 1.4238>
         <3.9063 5.424 -0.4356>
         <-3.9072 17.5907 3.6281>
         <-3.2314 16.7657 2.8649>
         <-3.1971 19.4502 0.4205>
         <-3.8666 20.7788 1.2006>
         <-3.4936 18.8399 -2.5452>
         <-3.3257 17.8094 -2.7333>
         <2.9755 5.7664 -0.2512>
        ]}}
      Random Seed=0
     )
     9=&SelectVectorFromArray(
      Array Sampling Mode=::NewEnumerator0
      Direct Array Index=?
      Vector Selection Array=DataInterface() {propertyValues {FloatData [<22.9754 12.4639 0.8774>, <23.1317 12.382 -2.0999>, <21.7803 12.7554 -2.9626>, <20.0873 13.4489 0.856>]}}
      Random Seed=0
     )
     Index=$Particles.SpawnGroup
     Num=::NewEnumerator8
    )
    Socket Index=$Particles.SpawnGroup
    Socket Reader=$System.Socket Reader
    Socket Sampling Mode=::NewEnumerator3
    VelocitySpeedLimit=false
   )
   &ScaleSpriteSize(Uniform Curve Sprite Scale=DataInterface() {propertyValues {Curve {keys [{arriveTangent=3.8099, time=0, value=0}, {arriveTangent=2.5, leaveTangent=2.5, time=0.3, value=1.2}, <at 0.4 1 RCIM_Cubic>]}}})
   &ScaleColor(
    Scale Alpha=&Multiply_Float(
     A=&Lerp_Float(
      A=1
      B=&UniformAOrBFloat(A=0, Evaluation Type=::NewEnumerator1, Recalculate Random Each Loop=true, Override Seed=false)
      Alpha=&FloatFromCurve(FloatCurve=DataInterface() {propertyValues {Curve {keys [<at 0.45 1>,<at 1 0>]}}})
     )
     B=&FNTimeOfDayScalar(Night Scale=0.4, Morning Scale=0.65, Evening Scale=0.75)
    )
    Scale RGB=&VectorFromFloat(Value=&FloatFromCurve(FloatCurve=DataInterface() {propertyValues {Curve {keys [{arriveTangent=-14.8544, interpMode=RCIM_Cubic, leaveTangent=-14.8544, tangentMode=RCTM_User, time=0.2, value=5}, <at 1 0>]}}}))
    ScaleA=true
    ScaleRGB=true
   )
   &KillParticles(Kill Particles=&Boolean_OrOperation(
     Bool1=&Boolean_AndOperation(
      Bool1=$User.bisHandHidden
      Bool2=&SetBoolByIntComparison(A=$Particles.PartGroup, B=0, Comparison Type=::NewEnumerator4)
     )
     Bool2=&Boolean_OrOperation(
      Bool1=&Boolean_AndOperation(
       Bool1=$User.bisFootHidden
       Bool2=&SetBoolByIntComparison(A=$Particles.PartGroup, B=4, Comparison Type=::NewEnumerator4)
      )
      Bool2=&Boolean_OrOperation(
       Bool1=&Boolean_AndOperation(
        Bool1=&Boolean_OrOperation(Bool1=$User.bisHeadHidden, Bool2=$System.bisHatHidden)
        Bool2=&SetBoolByIntComparison(A=$Particles.PartGroup, B=5, Comparison Type=::NewEnumerator4)
       )
       Bool2=&Boolean_AndOperation(
        Bool1=$System.bisNecklaceHidden
        Bool2=&SetBoolByIntComparison(A=$Particles.PartGroup, B=1, Comparison Type=::NewEnumerator4)
       )
      )
     )
    ))
  ]
  rendererClasses [&NiagaraSpriteRendererProperties]
  renderers [{rendererClass=&NiagaraSpriteRendererProperties}]
 }}
summary {
 systemName=NS_Character_FunkyFeathers_Coin_Idle
 emitters [{emitterName=DiamondBlings, rendererClasses [&NiagaraSpriteRendererProperties]}]
 userVariables [
  User.bisFootHidden {type=NiagaraBool}
  User.bisHandHidden {type=NiagaraBool}
  User.bisHeadHidden {type=NiagaraBool}
  User.HatEnabled {type=NiagaraFloat, defaultValue=1, description=""}
  User.NecklaceEnabled {type=NiagaraFloat, defaultValue=0, description=""}
 ]
}
